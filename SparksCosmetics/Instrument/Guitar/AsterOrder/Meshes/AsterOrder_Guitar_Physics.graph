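# Physics asset for the AsterOrder guitar cosmetic, setting up collision and joints for two loop chains attached to the guitar's handle.
# A sphere at loop_root is rotationally locked to the handle, which carries two capsules, and anchors a five-bone large loop and a three-bone small loop, all capsules.
# Each chain's first joint locks twist, while later joints use narrow twist limits and wider swing limits.
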PhysicsAsset /SparksCosmetics/Instrument/Guitar/AsterOrder/Meshes/AsterOrder_Guitar_Physics
bodies {
 handle [
  {shapeType=Capsule, length=37.7529, radius=2.1333, center=<0.5821 0.3103 34.535>, rotation=<rot 0 -90 0>}
  {shapeType=Capsule, length=1.5061, radius=1.5653, shapeName=handle_capsule, center=<-1.2296 -3.9893 59.3311>, rotation=<rot 0.0001 0 20.7948>}
 ]
 large_loop_01 [{shapeType=Capsule, length=1, radius=0.5, shapeName=large_loop_01_capsule, center=<0.6 0 0>, rotation=<rot 90 0 0>}]
 large_loop_02 [{shapeType=Capsule, length=1, radius=0.75, shapeName=large_loop_02_capsule, center=<1 0 0>, rotation=<rot 0 90 -89.9983>}]
 large_loop_03 [{shapeType=Capsule, length=3.2499, radius=1, shapeName=large_loop_03_capsule, center=<0 0 0.2>, rotation=<rot 90 -90 0>}]
 large_loop_04 [{shapeType=Capsule, length=3.1599, radius=1.2576, shapeName=large_loop_04_capsule, center=<0.6481 0 0.2>, rotation=<rot 90 -90 0>}]
 large_loop_05 [{shapeType=Capsule, length=2.5, radius=1.5, shapeName=large_loop_05_capsule, center=<0.5397 0 0.7>, rotation=<rot 90 90 0>}]
 loop_root [{shapeType=Sphere, radius=0.5, shapeName=loop_root_sphere}]
 small_loop_01 [{shapeType=Capsule, length=1, radius=0.5, shapeName=small_loop_01_capsule, center=<0.6 0 0>, rotation=<rot 0 -87.9352 -89.0184>}]
 small_loop_02 [{shapeType=Capsule, length=1, radius=0.75, shapeName=small_loop_02_capsule, center=<0.5 0 0>, rotation=<rot 90 0 0>}]
 small_loop_03 [{shapeType=Capsule, length=1.5, radius=0.75, shapeName=small_loop_03_capsule, center=<0.75 0 0>, rotation=<rot 90 0 0>}]
}
constraints [
 {bone1Name=large_loop_05, bone2Name=large_loop_04, swing1LimitDegrees=35, swing2LimitDegrees=45, twistLimitDegrees=5}
 {bone1Name=large_loop_01, bone2Name=loop_root, swing1LimitDegrees=20, swing2LimitDegrees=60, twistLimitDegrees=45, twistMotion=Locked}
 {bone1Name=large_loop_02, bone2Name=large_loop_01, swing1LimitDegrees=20, swing2LimitDegrees=30, twistLimitDegrees=5}
 {bone1Name=large_loop_03, bone2Name=large_loop_02, swing1LimitDegrees=30, swing2LimitDegrees=45, twistLimitDegrees=5}
 {bone1Name=large_loop_04, bone2Name=large_loop_03, swing1LimitDegrees=35, swing2LimitDegrees=45, twistLimitDegrees=5}
 {
  bone1Name=loop_root
  bone2Name=handle
  swing1LimitDegrees=0
  swing1Motion=Locked
  swing2LimitDegrees=60
  swing2Motion=Locked
  twistLimitDegrees=45
  twistMotion=Locked
 }
 {bone1Name=small_loop_01, bone2Name=loop_root, swing1LimitDegrees=20, swing2LimitDegrees=70, twistLimitDegrees=45, twistMotion=Locked}
 {bone1Name=small_loop_02, bone2Name=small_loop_01, swing1LimitDegrees=30, swing2LimitDegrees=45, twistLimitDegrees=10}
 {bone1Name=small_loop_03, bone2Name=small_loop_02, swing1LimitDegrees=40, swing2LimitDegrees=45, twistLimitDegrees=10}
]
