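# Animation Blueprint for the AsterOrder guitar cosmetic that adds physics-driven secondary motion to the instrument's skeleton.
# The input pose passes through a component-space RigidBody node using the guitar's physics asset, fixed downward gravity and damped, capped simulation-space motion.
# The update event switches a Gravity Override variable for skydiving or floating, but the active node does not read it.
# A second RigidBody node, simulating in the handle bone's space, is left unconnected.
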AnimBlueprint /SparksCosmetics/Instrument/Guitar/AsterOrder/Meshes/AsterOrder_Guitar_AnimBP
target_skeleton=&/AsterOrder_Guitar_Skeleton
graphs [
 AnimGraph {nodes [
   RigidBody!(ComponentPose, Alpha, OverrideWorldGravity:Vector="0,0,0", ExternalForce) -> (Pose) {
    bClampLinearTranslationLimitToRefPose=true
    bOverrideWorldGravity=true
    overridePhysicsAsset=&/AsterOrder_Guitar_Physics
    simulationSpace=BaseBoneSpace
    alphaBoolBlend {blendOption=Linear}
    baseBoneRef {boneName=handle}
    componentLinearAccScale=<0.25 0.25 0.25>
    componentLinearVelScale=<0.25 0.25 0.25>
    simSpaceSettings {worldAlpha=0.3}
   }
   n1 = RigidBody(ComponentPose=Local To Component(LocalPose=Input Pose()), OverrideWorldGravity="0,0,-980") {
    bOverrideWorldGravity=true
    overridePhysicsAsset=&AsterOrder_Guitar_Physics
    alphaBoolBlend {blendOption=Linear}
    baseBoneRef {boneName=handle}
    overrideWorldGravity {z=-980}
    simSpaceSettings {
     dampingAlpha=0.6
     maxAngularAcceleration=15
     maxAngularVelocity=6
     maxLinearAcceleration=800
     maxLinearVelocity=500
     velocityScaleZ=0.4
     worldAlpha=1
     externalLinearDragV=<0.1 0.1 0.1>
    }
   }
   Output Pose(Result=Component To Local(ComponentPose=@n1))
  ]}
 EventGraph {nodes [
   Event BlueprintUpdateAnimation()
   > n1 = Branch(Condition=OR Boolean(A=Get bIsSkydiveFloating(), B=Get bIsSkyDiving()))
   > Set Gravity Override(Gravity Override="0,-300,600")
   Set Gravity Override(execute=@n1.else, Gravity Override="0,0,-980")
   TryGetPawnOwner()
  ]}
]
variables [Gravity Override {type=Vector}]
